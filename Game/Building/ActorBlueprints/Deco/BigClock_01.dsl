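; Big clock decoration prop whose hands can switch to time-of-day materials, with audio that runs only while time updating is enabled.
; SetTickingEnabled, called from the construction script and cosmetic-only BeginPlay, stores the flag, swaps hand materials and toggles audio.
; Enabling also zeroes hour-hand rotation; disabling restores only the hour hand's material, leaving the minute hand's unchanged.
; UpdateTime only prints the current time of day.
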(event EventBeginPlay
  (Networking|CanExecuteCosmeticEvents
    (:True
      (CallFunction|SetTickingEnabled (Variables|Clock|GetBShouldUpdateTime)))))

(fn setTickingEnabled
  (params (enabled :type Boolean))
  (Variables|Clock|SetBShouldUpdateTime enabled)
  (Utilities|FlowControl|Sequence
    (:then_0
      (if (Variables|Clock|GetBShouldUpdateTime)
        (Rendering|Material|SetMaterial (Variables|BigClock_01|GetMinuteHandMesh) 0 "/Game/Environments/Props/Clocks/Materials/MI_ClockHand_TOD_Min.MI_ClockHand_TOD_Min")
        (Rendering|Material|SetMaterial (Variables|BigClock_01|GetHourHandMesh) 0 "/Game/Environments/Props/Clocks/Materials/MI_ClockHand_TOD_Hour.MI_ClockHand_TOD_Hour")
        (Transformation|SetRelativeRotation (Variables|BigClock_01|GetHourHandMesh) "0, 0, 0" false false)
        (else
          (Rendering|Material|SetMaterial (Variables|BigClock_01|GetHourHandMesh) 0 "/Game/Environments/Props/Clocks/Materials/MI_BigClock_01.MI_BigClock_01"))))
    (:then_1
      (Components|Activation|SetActive (Variables|Default|GetAudio) (Variables|Clock|GetBShouldUpdateTime) false))))

(fn UpdateTime
  (Development|PrintString (Utilities|String|ToString(Text) (Utilities|Text|ToText(Float) (TimeOfDay|GetTimeOfDay :!class /Script/FortniteGame.FortKismetLibrary) "HalfToEven" false false 1 324 0 2))))

(fn ConstructionScript
  (CallFunction|SetTickingEnabled (Variables|Clock|GetBShouldUpdateTime)))
