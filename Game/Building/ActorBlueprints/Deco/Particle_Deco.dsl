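; Particle decoration actor for buildings that shows a Cascade or fallback Niagara effect, with optional sound and spotlight.
; Construction script assigns the assets, builds the configurable light, and clamps the Cascade emitter's scale, optionally randomizing it through a sine hash of actor location.
; On BeginPlay the authority rolls the activation chance and silently kills the actor on failure.
; ActivateDecoEmitter runs on success, but its authority branch only validity-checks the particle.
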(event EventBeginPlay
  (Networking|SwitchHasAuthority
    (:Authority
      (bind _dropresult (Class|MissionBlueprintFunctionLibrary|DropChance (Math|Float|Clamp(Float) (* (Math|Float|Clamp(Float) (Variables|Default|GetEmitterActivateChance)) 1e+02) 0.0 1e+02)))
      (if _dropresult
        (CallFunction|ActivateDecoEmitter)
        (else
          (Gameplay|SilentDie self false))))))

(event Custom|ActivateDecoEmitter
  (Utilities|FlowControl|Sequence
    (:then_0
      (Utilities|IsValid (Variables|Default|GetSoundCue)
        (:Is%20Valid
          (Networking|SwitchHasAuthority
            (:Remote
              (Audio|Components|Audio|SetSound (Variables|Default|GetAudio) (Variables|Default|GetSoundCue)))))))
    (:then_1
      (Networking|SwitchHasAuthority
        (:Authority
          (Utilities|IsValid (Variables|Default|GetParticle)))))))

(fn MaterialEditorSine
  (params (LookUpValue :type Float%20%28double-precision%29) (Period :type Float%20%28double-precision%29))
  (return (Math|Trig|Sin(Degrees) (/ (* LookUpValue 360) Period))))

(fn PseudoRandomScale
  (params (ActorLocation :type Vector) (ParticleScale :type Vector))
  (return (Math|Vector|MakeVector (Math|Float|Clamp(Float) (+ (.x (Variables|Default|GetParticleScale)) (/ (Math|Float|NormalizetoRange (Math|Float|Absolute(Float) (Math|Float|Fraction (* (* (CallFunction|MaterialEditorSine (+ (+ (/ (Math|Conversions|ToFloat(Integer) (Math|Float|Ceil (.y (/ (+ ActorLocation 5.25) (Variables|Default|GetRandomValue_5.59614))))) (Variables|Default|GetRandomValue_5.59614)) (* (/ (Math|Conversions|ToFloat(Integer) (Math|Float|Ceil (.x (/ (+ ActorLocation 5.25) (Variables|Default|GetRandomValue_5.59614))))) (Variables|Default|GetRandomValue_5.59614)) 0.04523)) (* (CallFunction|MaterialEditorSine (+ (* (CallFunction|MaterialEditorSine (+ (/ (Math|Conversions|ToFloat(Integer) (Math|Float|Ceil (.x (/ (+ ActorLocation 5.25) (Variables|Default|GetRandomValue_5.59614))))) (Variables|Default|GetRandomValue_5.59614)) (* (/ (Math|Conversions|ToFloat(Integer) (Math|Float|Ceil (.y (/ (+ ActorLocation 5.25) (Variables|Default|GetRandomValue_5.59614))))) (Variables|Default|GetRandomValue_5.59614)) -0.21458)) 0.201836) 0.023) (+ (/ (Math|Conversions|ToFloat(Integer) (Math|Float|Ceil (.x (/ (+ ActorLocation 5.25) (Variables|Default|GetRandomValue_5.59614))))) (Variables|Default|GetRandomValue_5.59614)) (/ (Math|Conversions|ToFloat(Integer) (Math|Float|Ceil (.y (/ (+ ActorLocation 5.25) (Variables|Default|GetRandomValue_5.59614))))) (Variables|Default|GetRandomValue_5.59614)))) 0.145872) 0.01914)) 0.1) (CallFunction|MaterialEditorSine (+ (* (CallFunction|MaterialEditorSine (+ (/ (Math|Conversions|ToFloat(Integer) (Math|Float|Ceil (.x (/ (+ ActorLocation 5.25) (Variables|Default|GetRandomValue_5.59614))))) (Variables|Default|GetRandomValue_5.59614)) (* (/ (Math|Conversions|ToFloat(Integer) (Math|Float|Ceil (.y (/ (+ ActorLocation 5.25) (Variables|Default|GetRandomValue_5.59614))))) (Variables|Default|GetRandomValue_5.59614)) -0.21458)) 0.201836) 0.023) (+ (/ (Math|Conversions|ToFloat(Integer) (Math|Float|Ceil (.x (/ (+ ActorLocation 5.25) (Variables|Default|GetRandomValue_5.59614))))) (Variables|Default|GetRandomValue_5.59614)) (/ (Math|Conversions|ToFloat(Integer) (Math|Float|Ceil (.y (/ (+ ActorLocation 5.25) (Variables|Default|GetRandomValue_5.59614))))) (Variables|Default|GetRandomValue_5.59614)))) 0.145872)) 13.1145769))) 0.25 2.0) 1e+01)) 0.25 1e+01) (Math|Float|Clamp(Float) (+ (.y (Variables|Default|GetParticleScale)) (/ (Math|Float|NormalizetoRange (Math|Float|Absolute(Float) (Math|Float|Fraction (* (* (CallFunction|MaterialEditorSine (+ (+ (/ (Math|Conversions|ToFloat(Integer) (Math|Float|Ceil (.y (/ (+ ActorLocation 5.25) (Variables|Default|GetRandomValue_5.59614))))) (Variables|Default|GetRandomValue_5.59614)) (* (/ (Math|Conversions|ToFloat(Integer) (Math|Float|Ceil (.x (/ (+ ActorLocation 5.25) (Variables|Default|GetRandomValue_5.59614))))) (Variables|Default|GetRandomValue_5.59614)) 0.04523)) (* (CallFunction|MaterialEditorSine (+ (* (CallFunction|MaterialEditorSine (+ (/ (Math|Conversions|ToFloat(Integer) (Math|Float|Ceil (.x (/ (+ ActorLocation 5.25) (Variables|Default|GetRandomValue_5.59614))))) (Variables|Default|GetRandomValue_5.59614)) (* (/ (Math|Conversions|ToFloat(Integer) (Math|Float|Ceil (.y (/ (+ ActorLocation 5.25) (Variables|Default|GetRandomValue_5.59614))))) (Variables|Default|GetRandomValue_5.59614)) -0.21458)) 0.201836) 0.023) (+ (/ (Math|Conversions|ToFloat(Integer) (Math|Float|Ceil (.x (/ (+ ActorLocation 5.25) (Variables|Default|GetRandomValue_5.59614))))) (Variables|Default|GetRandomValue_5.59614)) (/ (Math|Conversions|ToFloat(Integer) (Math|Float|Ceil (.y (/ (+ ActorLocation 5.25) (Variables|Default|GetRandomValue_5.59614))))) (Variables|Default|GetRandomValue_5.59614)))) 0.145872) 0.01914)) 0.1) (CallFunction|MaterialEditorSine (+ (* (CallFunction|MaterialEditorSine (+ (/ (Math|Conversions|ToFloat(Integer) (Math|Float|Ceil (.x (/ (+ ActorLocation 5.25) (Variables|Default|GetRandomValue_5.59614))))) (Variables|Default|GetRandomValue_5.59614)) (* (/ (Math|Conversions|ToFloat(Integer) (Math|Float|Ceil (.y (/ (+ ActorLocation 5.25) (Variables|Default|GetRandomValue_5.59614))))) (Variables|Default|GetRandomValue_5.59614)) -0.21458)) 0.201836) 0.023) (+ (/ (Math|Conversions|ToFloat(Integer) (Math|Float|Ceil (.x (/ (+ ActorLocation 5.25) (Variables|Default|GetRandomValue_5.59614))))) (Variables|Default|GetRandomValue_5.59614)) (/ (Math|Conversions|ToFloat(Integer) (Math|Float|Ceil (.y (/ (+ ActorLocation 5.25) (Variables|Default|GetRandomValue_5.59614))))) (Variables|Default|GetRandomValue_5.59614)))) 0.145872)) 13.1145769))) 0.25 2.0) 1e+01)) 0.25 1e+01) (Math|Float|Clamp(Float) (+ (.z (Variables|Default|GetParticleScale)) (/ (Math|Float|NormalizetoRange (Math|Float|Absolute(Float) (Math|Float|Fraction (* (* (CallFunction|MaterialEditorSine (+ (+ (/ (Math|Conversions|ToFloat(Integer) (Math|Float|Ceil (.y (/ (+ ActorLocation 5.25) (Variables|Default|GetRandomValue_5.59614))))) (Variables|Default|GetRandomValue_5.59614)) (* (/ (Math|Conversions|ToFloat(Integer) (Math|Float|Ceil (.x (/ (+ ActorLocation 5.25) (Variables|Default|GetRandomValue_5.59614))))) (Variables|Default|GetRandomValue_5.59614)) 0.04523)) (* (CallFunction|MaterialEditorSine (+ (* (CallFunction|MaterialEditorSine (+ (/ (Math|Conversions|ToFloat(Integer) (Math|Float|Ceil (.x (/ (+ ActorLocation 5.25) (Variables|Default|GetRandomValue_5.59614))))) (Variables|Default|GetRandomValue_5.59614)) (* (/ (Math|Conversions|ToFloat(Integer) (Math|Float|Ceil (.y (/ (+ ActorLocation 5.25) (Variables|Default|GetRandomValue_5.59614))))) (Variables|Default|GetRandomValue_5.59614)) -0.21458)) 0.201836) 0.023) (+ (/ (Math|Conversions|ToFloat(Integer) (Math|Float|Ceil (.x (/ (+ ActorLocation 5.25) (Variables|Default|GetRandomValue_5.59614))))) (Variables|Default|GetRandomValue_5.59614)) (/ (Math|Conversions|ToFloat(Integer) (Math|Float|Ceil (.y (/ (+ ActorLocation 5.25) (Variables|Default|GetRandomValue_5.59614))))) (Variables|Default|GetRandomValue_5.59614)))) 0.145872) 0.01914)) 0.1) (CallFunction|MaterialEditorSine (+ (* (CallFunction|MaterialEditorSine (+ (/ (Math|Conversions|ToFloat(Integer) (Math|Float|Ceil (.x (/ (+ ActorLocation 5.25) (Variables|Default|GetRandomValue_5.59614))))) (Variables|Default|GetRandomValue_5.59614)) (* (/ (Math|Conversions|ToFloat(Integer) (Math|Float|Ceil (.y (/ (+ ActorLocation 5.25) (Variables|Default|GetRandomValue_5.59614))))) (Variables|Default|GetRandomValue_5.59614)) -0.21458)) 0.201836) 0.023) (+ (/ (Math|Conversions|ToFloat(Integer) (Math|Float|Ceil (.x (/ (+ ActorLocation 5.25) (Variables|Default|GetRandomValue_5.59614))))) (Variables|Default|GetRandomValue_5.59614)) (/ (Math|Conversions|ToFloat(Integer) (Math|Float|Ceil (.y (/ (+ ActorLocation 5.25) (Variables|Default|GetRandomValue_5.59614))))) (Variables|Default|GetRandomValue_5.59614)))) 0.145872)) 13.1145769))) 0.25 2.0) 1e+01)) 0.25 1e+01))))

(fn ConstructionScript
  (bind _decoemitter (Variables|Default|GetDecoEmitter))
  (bind _soundcue (Variables|Default|GetSoundCue))
  (bind _ns_particle (Variables|Default|GetNS_Particle))
  (bind _particle (Variables|Default|GetParticle))
  (Utilities|FlowControl|Sequence
    (:then_0
      (Utilities|IsValid _particle
        (:Is%20Valid
          (Effects|Components|ParticleSystem|SetTemplate _decoemitter _particle))
        (:Is%20Not%20Valid
          (Utilities|IsValid _ns_particle
            (:Is%20Valid
              (Niagara|SetNiagaraSystemAsset (Variables|Default|GetNS_DecoEmitter) _ns_particle true)))))
      (Transformation|SetWorldScale3D _decoemitter (Math|Vector|ClampVectorSize (Variables|Default|GetParticleScale) 0.25 1e+01))
      (if (Variables|Default|GetAllowRandomParticleScale)
        (Transformation|SetWorldScale3D (Variables|Default|GetDecoEmitter) (CallFunction|PseudoRandomScale (Transformation|GetActorLocation) (Variables|Default|GetParticleScale)))))
    (:then_1
      (Utilities|IsValid _soundcue
        (:Is%20Valid
          (Audio|Components|Audio|SetSound (Variables|Default|GetAudio) _soundcue))))
    (:then_2
      (if (Variables|Default|GetEnableSpotlight)
        (bind _returnvalue (AddComponent|NODE_AddSpotLightComponent-0 :RelativeTransform (Variables|Default|GetSpotLightRelativeTransformOffset)))
        (Rendering|Lighting|SetAttenuationRadius _returnvalue (Variables|Default|GetSpotLightAttenuationRadius))
        (Rendering|Components|Light|SetIntensity _returnvalue (Variables|Default|GetSpotlightIntensity))
        (Rendering|Components|Light|SetLightColor _returnvalue (Variables|Default|GetSpotLightColor) true)
        (Rendering|Lighting|SetInnerConeAngle _returnvalue (Variables|Default|GetSpotlightInnerConeAngle))
        (Rendering|Lighting|SetOuterConeAngle _returnvalue (Variables|Default|GetSpotlightOuterConeAngle))
        (Rendering|Components|Light|SetCastShadows _returnvalue (Variables|Default|GetSpotLightCastShadows))
        (Rendering|Components|Light|SetVolumetricScatteringIntensity _returnvalue (Variables|Default|GetSpotLightVolumetricScatteringIntensity) :!class /Script/Engine.LightComponent)))))
