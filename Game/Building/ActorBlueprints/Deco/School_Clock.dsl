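; School clock decoration building actor whose hour and minute hands display the in-game time of day.
; On BeginPlay, where cosmetic events can execute, a looping one-second timer calls UpdateTime.
; UpdateTime sets each hand mesh's relative rotation from GetTimeOfDay, using a twelve-hour cycle for hours and division by one for minutes.
; On death, it fades out the ambient audio loop if valid and playing.
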(event EventBeginPlay
  (Networking|CanExecuteCosmeticEvents
    (:True
      (Utilities|Time|SetTimerbyFunctionName 0 "UpdateTime" 1 true))))

(event EventOnDeathPlayEffects
  (params (Damage :type Float%20%28single-precision%29) (DamageTags :type Gameplay%20Tag%20Container%20Structure%20%28by%20ref%29) (Momentum :type Vector :default "0, 0, 0") (HitInfo :type Hit%20Result%20Structure%20%28by%20ref%29) (InstigatedBy :type Fort%20Pawn%20Object%20Reference) (DamageCauser :type Actor%20Object%20Reference) (EffectContext :type Gameplay%20Effect%20Context%20Handle%20Structure))
  (bind _ambientaudioloop (Variables|Default|GetAmbientAudioLoop))
  (Utilities|IsValid _ambientaudioloop
    (:Is%20Valid
      (if (Audio|Components|Audio|IsPlaying _ambientaudioloop)
        (Audio|Components|Audio|FadeOut _ambientaudioloop 0.2 0.0 "Linear" :!class /Script/Engine.AudioComponent)))))

(fn UpdateTime
  (Transformation|SetRelativeRotation (Variables|School_Clock|GetHourHandMesh) (Math|Rotator|MakeRotator (* (/ (TimeOfDay|GetTimeOfDay :!class /Script/FortniteGame.FortKismetLibrary) 12) 360)) false false)
  (Transformation|SetRelativeRotation (Variables|School_Clock|GetMinuteHandMesh) (Math|Rotator|MakeRotator (* (Math|Float|Division(WholeandRemainder) (TimeOfDay|GetTimeOfDay :!class /Script/FortniteGame.FortKismetLibrary) 1) 360)) false false))

(fn ConstructionScript)
